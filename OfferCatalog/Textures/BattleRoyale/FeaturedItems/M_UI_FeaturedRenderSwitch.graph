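# Translucent UI material for featured item renders that automatically cycles outfit images with an animated wipe, keeping each image's alpha.
# A time-driven, sliding and rotating UV gradient forms the wipe mask, while its narrow edge band plus a gated sine flash color-dodge brightens each texture.
# Fmod and If phase selectors pick the outgoing and incoming pair for each cycle.
# Static switch Use4TextureSwipe chooses looping through TextureA to TextureD or alternating between A and B.
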Material /OfferCatalog/Textures/BattleRoyale/FeaturedItems/M_UI_FeaturedRenderSwitch
graph {
 properties {
  bUsedWithStaticMesh=false
  blendMode=BLEND_Translucent
  materialDomain=MD_UI
  pixelDepthOffsetMode=PDOM_Legacy
  usedShadingModels=Default Lit
 }
 nodes [
  n1 = Multiply(A=Time()) {constB=0.425}
  n2 = Add(A=@n1) {constB=-0.75}
  n3 = Constant()
  n4 = LinearInterpolate(
   B=Sine(None=Multiply(A=@n2) {constB=4})
   Alpha=If(A=Sine(None=@n2), B=Constant() {r=-0.75}, "A > B"=@n3, "A == B"=@n3, "A < B"=Constant() {r=1})
  )
  n5 = Fmod(A=@n1, B=Constant() {r=1})
  n6 = Add(A=TextureCoordinate() {uTiling=0.5, vTiling=0.5}, B=LinearInterpolate(A=Constant() {r=-1.75}, B=Constant() {r=0.75}, Alpha=@n5))
  n7 = LinearInterpolate(A=Constant(), B=Constant() {r=-0.25}, Alpha=@n5)
  n8 = ComponentMask(None=Rotator(Coordinate=Add(A=@n6) {constB=-0.0312}, Time=@n7) {centerX=0, centerY=0, speed=6.2832}) {r=true}
  n9 = Multiply(A=ComponentMask(None=Rotator(Coordinate=@n6, Time=@n7) {centerX=0, centerY=0, speed=6.2832}) {r=true}) {constB=32}
  n10 = DotProduct(A=OneMinus(None=Multiply(A=@n8) {constB=32}), B=@n9)
  n11 = Clamp(None=Add(A=Clamp(None=@n4), B=Clamp(None=@n10))) {maxDefault=0.925}
  n12 = TextureSampleParameter2D() {parameterName=TextureA, texture=/OfferCatalog/Textures/BattleRoyale/FeaturedItems/Outfit/T-AthenaSoldiers-CID-093-Athena-Commando-M-Dinosaur}
  n13 = MaterialFunctionCall!(Blend=@n11, Base=@n12.RGB) {materialFunction=/Engine/Functions/Engine_MaterialFunctions03/Blends/Blend_ColorDodge}
  n14 = TextureSampleParameter2D() {parameterName=TextureB, texture=/OfferCatalog/Textures/BattleRoyale/FeaturedItems/Outfit/T-AthenaSoldiers-CID-132-Athena-Commando-M-Venus}
  n15 = MaterialFunctionCall(Blend=@n11, Base=@n14.RGB) {materialFunction=&Blend_ColorDodge}
  n16 = Clamp(None=@n9)
  n17 = LinearInterpolate(A=@n13.Result, B=@n15.Result, Alpha=@n16)
  n18 = TextureSampleParameter2D() {
   parameterName=TextureC
   texture=/OfferCatalog/Textures/BattleRoyale/FeaturedItems/Outfit/T-AthenaSoldiers-CID-137-Athena-Commando-F-Basketball
  }
  n19 = MaterialFunctionCall(Blend=@n11, Base=@n18.RGB) {materialFunction=&Blend_ColorDodge}
  n20 = Multiply(A=@n1) {constB=0.5}
  n21 = Constant() {r=1}
  n22 = Constant() {r=0.5}
  n23 = Constant()
  n24 = Constant() {r=1}
  n25 = If(A=Fmod(A=@n20, B=@n21), B=@n22, "A > B"=@n23, "A == B"=@n23, "A < B"=@n24)
  n26 = LinearInterpolate(A=@n17, B=LinearInterpolate(A=@n15.Result, B=@n19.Result, Alpha=@n16), Alpha=@n25)
  n27 = TextureSampleParameter2D() {
   parameterName=TextureD
   texture=/OfferCatalog/Textures/BattleRoyale/FeaturedItems/Outfit/T-AthenaSoldiers-CID-100-Athena-Commando-M-CuChulainn
  }
  n28 = MaterialFunctionCall(Blend=@n11, Base=@n27.RGB) {materialFunction=&Blend_ColorDodge}
  n29 = LinearInterpolate(A=LinearInterpolate(A=@n19.Result, B=@n28.Result, Alpha=@n16), B=LinearInterpolate(A=@n28.Result, B=@n13.Result, Alpha=@n16), Alpha=@n25)
  n30 = Add(A=@n20) {constB=0.5}
  n31 = If(A=Fmod(A=Multiply(A=@n30) {constB=0.5}, B=@n21), B=@n22, "A > B"=@n23, "A == B"=@n23, "A < B"=@n24)
  n32 = LinearInterpolate(A=@n12.A, B=@n14.A, Alpha=@n16)
  n33 = LinearInterpolate(A=@n32, B=LinearInterpolate(A=@n14.A, B=@n18.A, Alpha=@n16), Alpha=@n25)
  n34 = LinearInterpolate(A=LinearInterpolate(A=@n18.A, B=@n27.A, Alpha=@n16), B=LinearInterpolate(A=@n27.A, B=@n12.A, Alpha=@n16), Alpha=@n25)
  n35 = LinearInterpolate(A=@n15.Result, B=@n13.Result, Alpha=@n16)
  n36 = LinearInterpolate(A=@n14.A, B=@n12.A, Alpha=@n16)
  n37 = StaticSwitchParameter(
   True=AppendVector(A=LinearInterpolate(A=@n26, B=@n29, Alpha=@n31), B=LinearInterpolate(A=@n33, B=@n34, Alpha=@n31))
   False=AppendVector(A=LinearInterpolate(A=@n17, B=@n35, Alpha=@n25), B=LinearInterpolate(A=@n32, B=@n36, Alpha=@n25))
  ) {parameterName=Use4TextureSwipe}
  ComponentMask(None=@n37) {a=true}
  ComponentMask(None=@n37) {b=true, g=true, r=true}
 ]
}
